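# Niagara gameplay-cue effect for Hightower snow: a short snow burst from the owner followed by swirling, falling flakes.
# Two burst emitters (CPU and GPU) use decaying spawn curves and a local PosAndVel module driven by emitter age, alongside a vortex-driven SubUV puff layer.
# A delayed CPU/GPU pair fills a raised cylinder with flakes drifting under gravity and panning curl noise.
# Flakes animate a dynamic material parameter with a spiky, flickering curve over their life.
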NiagaraSystem /FNGameplayCues/Athena/Effects/Niagara/Snow/NS_hightowerSnow
dependencies {usedDataInterfaces [&NiagaraDataInterfaceCurve], usedDynamicInputs [], usedModules [&SystemState]}
emitters {
 NE_Empty {
  emitterUpdateScript [
   &EmitterState(Life Cycle Mode=::NewEnumerator1, Loop Behavior=::NewEnumerator1, Loop Duration=0.5, UseLoopDelay=false)
   &SpawnRate(
    SpawnRate=&FloatFromCurve(
     FloatCurve=DataInterface() {propertyValues {Curve {keys [<at 0 5000>, {arriveTangent=-5000, interpMode=RCIM_Cubic, leaveTangent=-5000, time=0.2705, value=1889.0771}, <at 1 0>]}}}
     CurveIndex=$Emitter.NormalizedLoopAge
     Scale Curve=0.75
    )
    Use Spawn Probability=false
   )
  ]
  particleSpawnScript [
   /Niagara/Modules/Spawn/Initialization/InitializeParticle(
    Color=<rgba 1 0.7816 0.6927 1>
    Lifetime=&UniformRangedFloat(Minimum=3, Maximum=5)
    Module.Write Color=true
    Module.Write Lifetime=true
    Module.Write Mass=true
    Module.Write Material Random=true
    Module.Write Position=true
    Module.Write Scale=false
    Module.Write SpriteRotation=true
    Module.Write SpriteSize=true
    Position=&Add_Vector(A=Linked() {linkedVariable {name=Engine.Owner.Position, type=NiagaraPosition}}, B=<0 0 -50>)
    Sprite Rotation=&UniformRangedFloat(Minimum=-360, Maximum=360)
    Sprite Size=&Vector2DFromFloat(Value=&FloatFromCurve(
      FloatCurve=DataInterface() {propertyValues {Curve {keys [<at 0 20 RCIM_Cubic>,<at 0.1869 5.8825>,<at 1 3 RCIM_Cubic>]}}}
      CurveIndex=&UniformRangedFloat(Minimum=0, Maximum=1)
     ))
   )
   &DynamicMaterialParameters(
    Index 0 Param 1=-0.1
    Index 0 Param 2=10
    Index 0 Param 4=&UniformRangedFloat(Minimum=0, Maximum=3)
    Module.Param0WriteEnabled=true
    Module.Param1WriteEnabled=true
    Module.Param2WriteEnabled=true
    Module.Param3WriteEnabled=true
   )
   &self:PosAndVel(
    Time:NiagaraFloat=&FloatFromCurve(FloatCurve=DataInterface() {propertyValues {Curve {keys [<at 0 0>,<at 0.5 500>]}}}, CurveIndex=$Emitter.Age)
    VelScale:NiagaraFloat=&FloatFromCurve(FloatCurve=DataInterface() {propertyValues {Curve {keys [<at 0 -1000>,<at 0.5 -2000>]}}}, CurveIndex=$Emitter.Age)
   ) {module=PosAndVel}
   &AddVelocity(Velocity=<0 0 100>)
   /SetVariables(Particles.RamdomFloat:NiagaraFloat=&UniformRangedFloat(Minimum=0, Maximum=1)) {isSetParameters=true, module=SetVariables_#1}
  ]
  particleUpdateScript [
   &ParticleState()
   &GravityForce(Gravity=<0 0 -98>)
   &Drag(
    Drag=&FloatFromCurve(FloatCurve=DataInterface() {propertyValues {Curve {keys [
         {arriveTangent=-3.5797, interpMode=RCIM_Cubic, leaveTangent=-3.5797, tangentMode=RCTM_User, time=0, value=1.5}
         {arriveTangent=-1.5, interpMode=RCIM_Cubic, leaveTangent=-1.5, time=0.3384, value=0.4446}
         <at 1 0>
        ]}}})
    Use Linear Drag=true
    Use Rotational Drag=false
   )
   &SolveForcesAndVelocity(Clamp Velocity=false, Limit Acceleration=false)
   &DynamicMaterialParameters(
    Index 0 Param 1=&FloatFromCurve(FloatCurve=DataInterface() {propertyValues {Curve {keys [<at -0.002 0.8557>,<at 0.1212 -0.1>,<at 1 -0.1>]}}})
    Index 0 Param 2=&FloatFromCurve(
     FloatCurve=DataInterface() {propertyValues {Curve {keys [
         <at 0 1.4645 RCIM_Cubic>
         <at 0.0293 100>
         <at 0.0953 14.9171>
         {arriveTangent=-91.6762, interpMode=RCIM_Cubic, leaveTangent=-91.6762, time=0.2117, value=50}
         {arriveTangent=-207.689, interpMode=RCIM_Cubic, leaveTangent=-207.689, time=0.2498, value=0.7502}
         {arriveTangent=-1, interpMode=RCIM_Cubic, leaveTangent=-1, time=0.4123, value=8.3434}
         {arriveTangent=-8.9117, interpMode=RCIM_Cubic, leaveTangent=-8.9117, time=0.4423, value=0.5577}
         {arriveTangent=-1, interpMode=RCIM_Cubic, leaveTangent=-1, time=0.677, value=5.9839}
         {arriveTangent=-9.7222, interpMode=RCIM_Cubic, leaveTangent=-9.7222, time=0.6991, value=0.3009}
         {arriveTangent=-1, interpMode=RCIM_Cubic, leaveTangent=-1, time=0.8291, value=4.5054}
         {arriveTangent=-15.637, interpMode=RCIM_Cubic, leaveTangent=-15.637, time=0.8675, value=0.1325}
         {arriveTangent=-1.5156, interpMode=RCIM_Cubic, leaveTangent=-1.5156, tangentMode=RCTM_User, time=0.9615, value=2.4354}
         {interpMode=RCIM_Cubic, leaveTangent=7.1537, time=1, value=0}
        ]}}}
     Scale Curve=10
    )
    Module.Param0WriteEnabled=true
    Module.Param1WriteEnabled=true
    Module.Param2WriteEnabled=false
    Module.Param3WriteEnabled=false
   ) {module=DynamicMaterialParameters001}
   &ScaleSpriteSize(Scale Factor=&Vector2DFromFloat(Value=&FloatFromCurve(FloatCurve=DataInterface() {propertyValues {Curve {keys [<at 0 0 RCIM_Cubic>,<at 0.1351 0.878 RCIM_Cubic>,<at 1 0 RCIM_Cubic>]}}})))
   &SpriteRotationRate(Rotation Rate=&Lerp_Float(A=-360, B=360, Alpha=$Particles.RamdomFloat))
   &CameraOffset(Camera Offset Amount=-20)
  ]
  rendererClasses [&NiagaraSpriteRendererProperties]
  renderers [{rendererClass=&NiagaraSpriteRendererProperties}]
 }
 NE_Empty001 {
  simTarget=GPUComputeSim
  emitterUpdateScript [
   &EmitterState(Life Cycle Mode=::NewEnumerator1, Loop Behavior=::NewEnumerator1, Loop Duration=0.5, UseLoopDelay=false)
   &SpawnRate(
    SpawnRate=&FloatFromCurve(
     FloatCurve=DataInterface() {propertyValues {Curve {keys [<at 0 5000>, {arriveTangent=-5000, interpMode=RCIM_Cubic, leaveTangent=-5000, time=0.2705, value=1889.0771}, <at 1 0>]}}}
     CurveIndex=$Emitter.NormalizedLoopAge
     Scale Curve=8
    )
    Use Spawn Probability=false
   )
  ]
  particleSpawnScript [
   &InitializeParticle(
    Color=<rgba 1.4375 1.8525 2 1>
    Lifetime=&UniformRangedFloat(Minimum=0.1, Maximum=1.5)
    Module.Write Color=true
    Module.Write Lifetime=true
    Module.Write Mass=true
    Module.Write Material Random=true
    Module.Write Position=true
    Module.Write Scale=false
    Module.Write SpriteRotation=true
    Module.Write SpriteSize=true
    Position=&Add_Vector(A=Linked() {linkedVariable {name=Engine.Owner.Position, type=NiagaraPosition}}, B=<0 0 -50>)
    Sprite Rotation=&UniformRangedFloat(Minimum=-360, Maximum=360)
    Sprite Size=&Vector2DFromFloat(Value=&FloatFromCurve(
      FloatCurve=DataInterface() {propertyValues {Curve {keys [<at 0 12 RCIM_Cubic>,<at 0.1869 5>,<at 1 4 RCIM_Cubic>]}}}
      CurveIndex=&UniformRangedFloat(Minimum=0, Maximum=1)
     ))
   )
   &DynamicMaterialParameters(
    Index 0 Param 2=10
    Index 0 Param 4=&UniformRangedFloat(Minimum=0, Maximum=3)
    Module.Param0WriteEnabled=true
    Module.Param1WriteEnabled=true
    Module.Param2WriteEnabled=true
    Module.Param3WriteEnabled=true
   )
   &self:PosAndVel(
    Time:NiagaraFloat=&FloatFromCurve(FloatCurve=DataInterface() {propertyValues {Curve {keys [<at 0 0>,<at 0.5 500>]}}}, CurveIndex=$Emitter.Age)
    VelScale:NiagaraFloat=&FloatFromCurve(FloatCurve=DataInterface() {propertyValues {Curve {keys [<at 0 -1000>,<at 0.5 -2000>]}}}, CurveIndex=$Emitter.Age)
   ) {module=PosAndVel}
   &AddVelocity(Velocity=<0 0 100>)
   /SetVariables(Particles.RamdomFloat:NiagaraFloat=&UniformRangedFloat(Minimum=0, Maximum=1)) {isSetParameters=true, module=SetVariables_#1}
  ]
  particleUpdateScript [
   &ParticleState()
   &GravityForce(Gravity=<0 0 -98>)
   &Drag(
    Drag=&FloatFromCurve(FloatCurve=DataInterface() {propertyValues {Curve {keys [
         {arriveTangent=-3.5797, interpMode=RCIM_Cubic, leaveTangent=-3.5797, tangentMode=RCTM_User, time=0, value=1.5}
         {arriveTangent=-1.5, interpMode=RCIM_Cubic, leaveTangent=-1.5, time=0.3384, value=0.4446}
         <at 1 0>
        ]}}})
    Use Linear Drag=true
    Use Rotational Drag=false
   )
   &SolveForcesAndVelocity(Clamp Velocity=false, Limit Acceleration=false)
   &DynamicMaterialParameters(
    Index 0 Param 2=&FloatFromCurve(
     FloatCurve=DataInterface() {propertyValues {Curve {keys [
         <at 0 1.4645 RCIM_Cubic>
         <at 0.0293 100>
         <at 0.0953 14.9171>
         {arriveTangent=-91.6762, interpMode=RCIM_Cubic, leaveTangent=-91.6762, time=0.2117, value=50}
         {arriveTangent=-207.689, interpMode=RCIM_Cubic, leaveTangent=-207.689, time=0.2498, value=0.7502}
         {arriveTangent=-1, interpMode=RCIM_Cubic, leaveTangent=-1, time=0.4123, value=8.3434}
         {arriveTangent=-8.9117, interpMode=RCIM_Cubic, leaveTangent=-8.9117, time=0.4423, value=0.5577}
         {arriveTangent=-1, interpMode=RCIM_Cubic, leaveTangent=-1, time=0.677, value=5.9839}
         {arriveTangent=-9.7222, interpMode=RCIM_Cubic, leaveTangent=-9.7222, time=0.6991, value=0.3009}
         {arriveTangent=-1, interpMode=RCIM_Cubic, leaveTangent=-1, time=0.8291, value=4.5054}
         {arriveTangent=-15.637, interpMode=RCIM_Cubic, leaveTangent=-15.637, time=0.8675, value=0.1325}
         {arriveTangent=-1.5156, interpMode=RCIM_Cubic, leaveTangent=-1.5156, tangentMode=RCTM_User, time=0.9615, value=2.4354}
         {interpMode=RCIM_Cubic, leaveTangent=7.1537, time=1, value=0}
        ]}}}
     Scale Curve=10
    )
    Module.Param0WriteEnabled=false
    Module.Param1WriteEnabled=true
    Module.Param2WriteEnabled=false
    Module.Param3WriteEnabled=false
   ) {module=DynamicMaterialParameters001}
   &ScaleSpriteSize(Scale Factor=&Vector2DFromFloat(Value=&FloatFromCurve(FloatCurve=DataInterface() {propertyValues {Curve {keys [<at 0 0 RCIM_Cubic>,<at 0.1351 0.878 RCIM_Cubic>,<at 1 0 RCIM_Cubic>]}}})))
   &SpriteRotationRate(Rotation Rate=&Lerp_Float(A=-360, B=360, Alpha=$Particles.RamdomFloat))
   &CameraOffset(Camera Offset Amount=-20)
  ]
  rendererClasses [&NiagaraSpriteRendererProperties]
  renderers [{rendererClass=&NiagaraSpriteRendererProperties}]
 }
 NE_Empty002 {
  emitterUpdateScript [
   &EmitterState(Life Cycle Mode=::NewEnumerator1, Loop Behavior=::NewEnumerator1, Loop Delay=0.45, Loop Duration=1, UseLoopDelay=true)
   &SpawnRate(SpawnRate=100, Use Spawn Probability=false)
  ]
  particleSpawnScript [
   &InitializeParticle(
    Color=<rgba 1 0.7816 0.6927 1>
    Lifetime=&UniformRangedFloat(Minimum=3, Maximum=5)
    Module.Write Color=true
    Module.Write Lifetime=true
    Module.Write Mass=true
    Module.Write Material Random=true
    Module.Write Position=false
    Module.Write Scale=false
    Module.Write SpriteRotation=true
    Module.Write SpriteSize=true
    Position=&SimulationPosition()
    Sprite Rotation=&UniformRangedFloat(Minimum=-360, Maximum=360)
    Sprite Size=&Vector2DFromFloat(Value=&FloatFromCurve(
      FloatCurve=DataInterface() {propertyValues {Curve {keys [<at 0 20 RCIM_Cubic>,<at 0.1869 5.8825>,<at 1 3 RCIM_Cubic>]}}}
      CurveIndex=&UniformRangedFloat(Minimum=0, Maximum=1)
      Scale Curve=2
     ))
   )
   &CylinderLocation(
    Cylinder Radius=300
    MaskSpawnGroup=false
    Module.Enable Lathe Profile=false
    Module.Override Randomness=false
    Module.Surface Only=false
    Offset=<0 0 300>
   )
   &DynamicMaterialParameters(
    Index 0 Param 1=-0.5
    Index 0 Param 2=10
    Index 0 Param 4=&UniformRangedFloat(Minimum=0, Maximum=3)
    Module.Param0WriteEnabled=true
    Module.Param1WriteEnabled=true
    Module.Param2WriteEnabled=true
    Module.Param3WriteEnabled=true
   )
   /SetVariables(Particles.RamdomFloat:NiagaraFloat=&UniformRangedFloat(Minimum=0, Maximum=1)) {isSetParameters=true, module=SetVariables_#1}
  ]
  particleUpdateScript [
   &ParticleState()
   &GravityForce(Gravity=<0 0 -98>)
   &Drag(Use Linear Drag=true, Use Rotational Drag=false)
   /Niagara/Modules/Update/Forces/CurlNoiseForce(Noise Strength=200, Pan Noise=true, Pan Noise Field=<0 0 10>)
   &SolveForcesAndVelocity(Clamp Velocity=false, Limit Acceleration=false)
   &DynamicMaterialParameters(
    Index 0 Param 2=&FloatFromCurve(
     FloatCurve=DataInterface() {propertyValues {Curve {keys [
         <at 0 1.4645 RCIM_Cubic>
         <at 0.0293 100>
         <at 0.0953 14.9171>
         {arriveTangent=-91.6762, interpMode=RCIM_Cubic, leaveTangent=-91.6762, time=0.2117, value=50}
         {arriveTangent=-207.689, interpMode=RCIM_Cubic, leaveTangent=-207.689, time=0.2498, value=0.7502}
         {arriveTangent=-1, interpMode=RCIM_Cubic, leaveTangent=-1, time=0.4123, value=8.3434}
         {arriveTangent=-8.9117, interpMode=RCIM_Cubic, leaveTangent=-8.9117, time=0.4423, value=0.5577}
         {arriveTangent=-1, interpMode=RCIM_Cubic, leaveTangent=-1, time=0.677, value=5.9839}
         {arriveTangent=-9.7222, interpMode=RCIM_Cubic, leaveTangent=-9.7222, time=0.6991, value=0.3009}
         {arriveTangent=-1, interpMode=RCIM_Cubic, leaveTangent=-1, time=0.8291, value=4.5054}
         {arriveTangent=-15.637, interpMode=RCIM_Cubic, leaveTangent=-15.637, time=0.8675, value=0.1325}
         {arriveTangent=-1.5156, interpMode=RCIM_Cubic, leaveTangent=-1.5156, tangentMode=RCTM_User, time=0.9615, value=2.4354}
         {interpMode=RCIM_Cubic, leaveTangent=7.1537, time=1, value=0}
        ]}}}
     Scale Curve=50
    )
    Module.Param0WriteEnabled=false
    Module.Param1WriteEnabled=true
    Module.Param2WriteEnabled=false
    Module.Param3WriteEnabled=false
   ) {module=DynamicMaterialParameters001}
   &ScaleSpriteSize(Scale Factor=&Vector2DFromFloat(Value=&FloatFromCurve(FloatCurve=DataInterface() {propertyValues {Curve {keys [<at 0 0 RCIM_Cubic>,<at 0.1351 0.878 RCIM_Cubic>,<at 1 0 RCIM_Cubic>]}}})))
   &SpriteRotationRate(Rotation Rate=&Lerp_Float(A=-120, B=120, Alpha=$Particles.RamdomFloat))
   &CameraOffset(Camera Offset Amount=-20)
  ]
  rendererClasses [&NiagaraSpriteRendererProperties]
  renderers [{rendererClass=&NiagaraSpriteRendererProperties}]
 }
 NE_Empty003 {
  simTarget=GPUComputeSim
  emitterUpdateScript [
   &EmitterState(Life Cycle Mode=::NewEnumerator1, Loop Behavior=::NewEnumerator1, Loop Delay=0.45, Loop Duration=1, UseLoopDelay=true)
   &SpawnRate(SpawnRate=500, Use Spawn Probability=false)
  ]
  particleSpawnScript [
   &InitializeParticle(
    Color=<rgba 1.2031 1.427 1.5 1>
    Lifetime=&UniformRangedFloat(Minimum=3, Maximum=5)
    Module.Write Color=true
    Module.Write Lifetime=true
    Module.Write Mass=true
    Module.Write Material Random=true
    Module.Write Position=false
    Module.Write Scale=false
    Module.Write SpriteRotation=true
    Module.Write SpriteSize=true
    Position=&SimulationPosition()
    Sprite Rotation=&UniformRangedFloat(Minimum=-360, Maximum=360)
    Sprite Size=&Vector2DFromFloat(Value=&FloatFromCurve(
      FloatCurve=DataInterface() {propertyValues {Curve {keys [<at 0 10 RCIM_Cubic>,<at 0.1869 5>,<at 1 3 RCIM_Cubic>]}}}
      CurveIndex=&UniformRangedFloat(Minimum=0, Maximum=1)
      Scale Curve=2
     ))
   )
   &CylinderLocation(
    Cylinder Radius=300
    MaskSpawnGroup=false
    Module.Enable Lathe Profile=false
    Module.Override Randomness=false
    Module.Surface Only=false
    Offset=<0 0 300>
   )
   &DynamicMaterialParameters(
    Index 0 Param 1=-0.5
    Index 0 Param 2=10
    Index 0 Param 4=&UniformRangedFloat(Minimum=0, Maximum=3)
    Module.Param0WriteEnabled=true
    Module.Param1WriteEnabled=true
    Module.Param2WriteEnabled=true
    Module.Param3WriteEnabled=true
   )
   /SetVariables(Particles.RamdomFloat:NiagaraFloat=&UniformRangedFloat(Minimum=0, Maximum=1)) {isSetParameters=true, module=SetVariables_#1}
  ]
  particleUpdateScript [
   &ParticleState()
   &GravityForce(Gravity=<0 0 -98>)
   &Drag(Use Linear Drag=true, Use Rotational Drag=false)
   &CurlNoiseForce(Noise Strength=200, Pan Noise=true, Pan Noise Field=<0 0 10>)
   &SolveForcesAndVelocity(Clamp Velocity=false, Limit Acceleration=false)
   &DynamicMaterialParameters(
    Index 0 Param 2=&FloatFromCurve(
     FloatCurve=DataInterface() {propertyValues {Curve {keys [
         <at 0 1.4645 RCIM_Cubic>
         <at 0.0293 100>
         <at 0.0953 14.9171>
         {arriveTangent=-91.6762, interpMode=RCIM_Cubic, leaveTangent=-91.6762, time=0.2117, value=50}
         {arriveTangent=-207.689, interpMode=RCIM_Cubic, leaveTangent=-207.689, time=0.2498, value=0.7502}
         {arriveTangent=-1, interpMode=RCIM_Cubic, leaveTangent=-1, time=0.4123, value=8.3434}
         {arriveTangent=-8.9117, interpMode=RCIM_Cubic, leaveTangent=-8.9117, time=0.4423, value=0.5577}
         {arriveTangent=-1, interpMode=RCIM_Cubic, leaveTangent=-1, time=0.677, value=5.9839}
         {arriveTangent=-9.7222, interpMode=RCIM_Cubic, leaveTangent=-9.7222, time=0.6991, value=0.3009}
         {arriveTangent=-1, interpMode=RCIM_Cubic, leaveTangent=-1, time=0.8291, value=4.5054}
         {arriveTangent=-15.637, interpMode=RCIM_Cubic, leaveTangent=-15.637, time=0.8675, value=0.1325}
         {arriveTangent=-1.5156, interpMode=RCIM_Cubic, leaveTangent=-1.5156, tangentMode=RCTM_User, time=0.9615, value=2.4354}
         {interpMode=RCIM_Cubic, leaveTangent=7.1537, time=1, value=0}
        ]}}}
     Scale Curve=50
    )
    Module.Param0WriteEnabled=false
    Module.Param1WriteEnabled=true
    Module.Param2WriteEnabled=false
    Module.Param3WriteEnabled=false
   ) {module=DynamicMaterialParameters001}
   &ScaleSpriteSize(Scale Factor=&Vector2DFromFloat(Value=&FloatFromCurve(FloatCurve=DataInterface() {propertyValues {Curve {keys [<at 0 0 RCIM_Cubic>,<at 0.1351 0.878 RCIM_Cubic>,<at 1 0 RCIM_Cubic>]}}})))
   &SpriteRotationRate(Rotation Rate=&Lerp_Float(A=-120, B=120, Alpha=$Particles.RamdomFloat))
   &CameraOffset(Camera Offset Amount=-20)
  ]
  rendererClasses [&NiagaraSpriteRendererProperties]
  renderers [{rendererClass=&NiagaraSpriteRendererProperties}]
 }
 NE_Empty004 {
  emitterUpdateScript [
   &EmitterState(Life Cycle Mode=::NewEnumerator1, Loop Behavior=::NewEnumerator1, Loop Duration=0.2, UseLoopDelay=true)
   &SpawnRate(
    SpawnRate=&FloatFromCurve(
     FloatCurve=DataInterface() {propertyValues {Curve {keys [<at 0 5000>, {arriveTangent=-5000, interpMode=RCIM_Cubic, leaveTangent=-5000, time=0.2705, value=1889.0771}, <at 1 0>]}}}
     CurveIndex=$Emitter.NormalizedLoopAge
     Scale Curve=0.07
    )
    Use Spawn Probability=false
   )
  ]
  particleSpawnScript [
   &InitializeParticle(
    Color=<rgba 1 1 1 0.75>
    Lifetime=&UniformRangedFloat(Minimum=1, Maximum=3)
    Module.Write Color=true
    Module.Write Lifetime=true
    Module.Write Mass=true
    Module.Write Material Random=true
    Module.Write Position=true
    Module.Write Scale=false
    Module.Write SpriteRotation=true
    Module.Write SpriteSize=true
    Position=&Add_Vector(A=Linked() {linkedVariable {name=Engine.Owner.Position, type=NiagaraPosition}}, B=<0 0 -50>)
    Sprite Rotation=&UniformRangedFloat(Minimum=-360, Maximum=360)
    Sprite Size=&Vector2DFromFloat(Value=&UniformRangedFloat(Minimum=80, Maximum=100))
   )
   &DynamicMaterialParameters(
    Index 0 Param 1=3
    Module.Param0WriteEnabled=true
    Module.Param1WriteEnabled=false
    Module.Param2WriteEnabled=false
    Module.Param3WriteEnabled=false
   )
   &self:PosAndVel(
    Time:NiagaraFloat=&FloatFromCurve(FloatCurve=DataInterface() {propertyValues {Curve {keys [<at 0 0>,<at 0.5 200>]}}}, CurveIndex=$Emitter.Age)
    VelScale:NiagaraFloat=&FloatFromCurve(FloatCurve=DataInterface() {propertyValues {Curve {keys [<at 0 -650>,<at 0.5 -650>]}}}, CurveIndex=$Emitter.Age)
   ) {module=PosAndVel}
   &AddVelocity(Velocity=<0 0 100>)
   /SetVariables(Particles.RamdomFloat:NiagaraFloat=&UniformRangedFloat(Minimum=0, Maximum=1)) {isSetParameters=true, module=SetVariables_#1}
  ]
  particleUpdateScript [
   &ParticleState()
   &GravityForce(Gravity=<0 0 -50>)
   &Drag(Drag=3, Use Linear Drag=true, Use Rotational Drag=false)
   &VortexForce(Vortex Force Amount=220)
   &SolveForcesAndVelocity(Clamp Velocity=false, Limit Acceleration=false)
   &ScaleSpriteSize(Scale Factor=&Vector2DFromFloat(Value=&FloatFromCurve(FloatCurve=DataInterface() {propertyValues {Curve {keys [<at 0 0 RCIM_Cubic>, {arriveTangent=1.5, interpMode=RCIM_Cubic, leaveTangent=1.5, time=0.1351, value=0.878}, <at 1 1.5 RCIM_Cubic>]}}})))
   &SpriteRotationRate(Rotation Rate=&Lerp_Float(A=-10, B=10, Alpha=$Particles.RamdomFloat))
   &CameraOffset(Camera Offset Amount=20)
   &ScaleColor(
    Scale Alpha=&FloatFromCurve(FloatCurve=DataInterface() {propertyValues {Curve {keys [
         <at 0 0 RCIM_Cubic>
         {arriveTangent=-2.2155, interpMode=RCIM_Cubic, leaveTangent=-2.2155, tangentMode=RCTM_User, time=0.2186, value=0.8592}
         <at 1 0 RCIM_Cubic>
        ]}}})
    ScaleA=true
    ScaleRGB=true
   )
   /Niagara/Modules/Update/SubUV/SubUVAnimation(Number Of Frames=64)
  ]
  rendererClasses [&NiagaraSpriteRendererProperties]
  renderers [{rendererClass=&NiagaraSpriteRendererProperties}]
 }
}
summary {
 systemName=NS_hightowerSnow
 emitters [
  {emitterName=NE_Empty, rendererClasses [&NiagaraSpriteRendererProperties]}
  {emitterName=NE_Empty001, simTarget=GPUComputeSim, rendererClasses [&NiagaraSpriteRendererProperties]}
  {emitterName=NE_Empty002, rendererClasses [&NiagaraSpriteRendererProperties]}
  {emitterName=NE_Empty003, simTarget=GPUComputeSim, rendererClasses [&NiagaraSpriteRendererProperties]}
  {emitterName=NE_Empty004, rendererClasses [&NiagaraSpriteRendererProperties]}
 ]
 userVariables []
}
